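# Looping ambient SoundCue for the rumbling underside of the Sky Island POI.
# The full mix layers sub rumble, glowing rock, crumble and rattle loops with randomly delayed rumble, metal, wire and debris one-shots across Far, Med and Detail attenuations.
# A quality node gives its first two slots that mix and its last two only glowing-rock and sub loops.
# Nested looping envelopers slowly swell volume and bend pitch to vary the loops.
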SoundCue /Game/Sounds/Asteria/POI/SkyIsland/Spots/Spot_SkyIsland_Rumble_Loop_Cue
root=Enveloper_3
nodes [
 {
  id=Attenuation_0
  children [SoundClass_2]
  properties {attenuationSettings=&/Spot_Fortnitanium_SkyIsland_Underside_Med_Att}
 }
 {
  id=Attenuation_2
  children [SoundClass_3]
  properties {attenuationSettings=&/Spot_Fortnitanium_SkyIsland_Underside_Detail_Att}
 }
 {
  id=Attenuation_3
  children [SoundClass_0]
  properties {attenuationSettings=&/Spot_Fortnitanium_SkyIsland_Underside_Far_Att}
 }
 {
  id=Attenuation_4
  children [SoundClass_4]
  properties {attenuationSettings=&Spot_Fortnitanium_SkyIsland_Underside_Far_Att}
 }
 {
  id=Attenuation_5
  children [SoundClass_5]
  properties {attenuationSettings=&Spot_Fortnitanium_SkyIsland_Underside_Med_Att}
 }
 {id=Delay_0, children [Random_0], properties {delayMax=12, delayMin=3}}
 {id=Delay_1, children [Modulator_1], properties {delayMax=18, delayMin=6}}
 {id=Delay_2, children [Random_3], properties {delayMax=12, delayMin=3}}
 {id=Delay_3, children [Modulator_3], properties {delayMax=25, delayMin=6}}
 {
  id=Enveloper_0
  children [QualityLevel_1]
  properties {
   bLoop=true
   bLoopIndefinitely=true
   loopEnd=30
   pitchCurve {editorCurveData {keys [{arriveTangent=-0.0013, interpMode=RCIM_Cubic, time=0}, <at 15 0.95 RCIM_Cubic>, {interpMode=RCIM_Cubic, leaveTangent=0.0013, time=30}]}}
   volumeCurve {editorCurveData {keys [<at 0 1>]}}
  }
 }
 {
  id=Enveloper_1
  children [Modulator_0]
  properties {
   bLoop=true
   bLoopIndefinitely=true
   loopEnd=30
   pitchCurve {editorCurveData {keys [{time=0}]}}
   volumeCurve {editorCurveData {keys [
      {arriveTangent=-0.01, interpMode=RCIM_Cubic, time=0, value=1}
      <at 15 0.7 RCIM_Cubic>
      {interpMode=RCIM_Cubic, leaveTangent=0.01, time=30, value=1}
     ]}}
  }
 }
 {
  id=Enveloper_2
  children [WavePlayer_10]
  properties {
   bLoop=true
   bLoopIndefinitely=true
   loopEnd=30
   pitchCurve {editorCurveData {keys [{time=0}]}}
   volumeCurve {editorCurveData {keys [
      {arriveTangent=-0.01, interpMode=RCIM_Cubic, time=0, value=1}
      <at 15 0.7 RCIM_Cubic>
      {interpMode=RCIM_Cubic, leaveTangent=0.01, time=30, value=1}
     ]}}
  }
 }
 {
  id=Enveloper_3
  children [Enveloper_0]
  properties {
   bLoop=true
   bLoopIndefinitely=true
   loopEnd=20
   pitchCurve {editorCurveData {keys [{time=0}]}}
   volumeCurve {editorCurveData {keys [
      {arriveTangent=-0.01, interpMode=RCIM_Cubic, time=0, value=1}
      <at 10 0.5 RCIM_Cubic>
      {interpMode=RCIM_Cubic, leaveTangent=0.01, time=20, value=1}
     ]}}
  }
 }
 {
  id=Enveloper_4
  children [Enveloper_1]
  properties {
   bLoop=true
   bLoopIndefinitely=true
   loopEnd=60
   pitchCurve {editorCurveData {keys [
      {arriveTangent=-0.0013, interpMode=RCIM_Cubic, time=0}
      {arriveTangent=-0.0004, interpMode=RCIM_Cubic, leaveTangent=-0.0004, tangentMode=RCTM_User, time=15}
      <at 30 0.95 RCIM_Cubic>
      {time=45}
      {interpMode=RCIM_Cubic, leaveTangent=0.0013, time=60}
     ]}}
   volumeCurve {editorCurveData {keys [<at 0 1>]}}
  }
 }
 {
  id=Enveloper_5
  children [Modulator_2]
  properties {
   bLoop=true
   bLoopIndefinitely=true
   loopEnd=30
   pitchCurve {editorCurveData {keys [{time=0}]}}
   volumeCurve {editorCurveData {keys [
      {arriveTangent=-0.01, interpMode=RCIM_Cubic, time=0, value=1}
      <at 15 0.5 RCIM_Cubic>
      {interpMode=RCIM_Cubic, leaveTangent=0.01, time=30, value=1}
     ]}}
  }
 }
 {id=Looping_0, children [Delay_0]}
 {id=Looping_1, children [Delay_1]}
 {id=Looping_2, children [Delay_2]}
 {id=Looping_3, children [Delay_3]}
 {id=Mixer_0, children [Attenuation_3, Attenuation_5, Attenuation_2, Attenuation_0], properties {inputVolume [1,1.6,2,1.5]}}
 {id=Mixer_1, children [Looping_0,WavePlayer_8], properties {inputVolume [0.4,0.8]}}
 {id=Mixer_2, children [Enveloper_2,Looping_1,Looping_3], properties {inputVolume [1.4,0.6,0.2]}}
 {id=Mixer_3, children [WavePlayer_20, WavePlayer_21], properties {inputVolume [1,1]}}
 {id=Mixer_4, children [Looping_2], properties {inputVolume [0.1]}}
 {id=Mixer_5, children [Mixer_2,WavePlayer_9], properties {inputVolume [1,1.5]}}
 {id=Mixer_6, children [Random_5], properties {inputVolume [0.7]}}
 {id=Mixer_7, children [Random_7], properties {inputVolume [0.7]}}
 {id=Modulator_0, children [WavePlayer_0]}
 {id=Modulator_1, children [Random_2], properties {pitchMax=1.1, volumeMax=1.1, volumeMin=0.9}}
 {id=Modulator_2, children [Mixer_3]}
 {id=Modulator_3, children [Random_4], properties {pitchMax=0.9, pitchMin=0.9}}
 {id=QualityLevel_1, children [Mixer_0, Mixer_0, Attenuation_4, Attenuation_4]}
 {id=Random_0, children [WavePlayer_1,WavePlayer_2,WavePlayer_5,WavePlayer_6], properties {weights [1,1,1,1]}}
 {id=Random_1, children [WavePlayer_22, WavePlayer_23, WavePlayer_24, WavePlayer_25], properties {weights [1,1,1,1]}}
 {id=Random_2, children [WavePlayer_11, WavePlayer_12, WavePlayer_13, WavePlayer_14, WavePlayer_15], properties {weights [1,1,1,1,1]}}
 {id=Random_3, children [Mixer_7,Random_1,Mixer_6,Random_6], properties {weights [1,1,1,1]}}
 {
  id=Random_4
  children [WavePlayer_3, WavePlayer_4, WavePlayer_7, WavePlayer_16, WavePlayer_17, WavePlayer_18, WavePlayer_19, WavePlayer_26]
  properties {weights [1,1,1,1,1,1,1,1]}
 }
 {id=Random_5, children [WavePlayer_29, WavePlayer_30, WavePlayer_31, WavePlayer_32, WavePlayer_33], properties {weights [1,1,1,1,1]}}
 {id=Random_6, children [WavePlayer_34, WavePlayer_35, WavePlayer_36, WavePlayer_37], properties {weights [1,1,1,1]}}
 {id=Random_7, children [WavePlayer_38, WavePlayer_39, WavePlayer_40, WavePlayer_41], properties {weights [1,1,1,1]}}
 {id=SoundClass_0, children [Mixer_1], properties {soundClassOverride=&Fort_Gadgets}}
 {id=SoundClass_2, children [Mixer_5], properties {soundClassOverride=&Fort_Ambient_Outdoors_Athena}}
 {id=SoundClass_3, children [Mixer_4], properties {soundClassOverride=&Fort_Ambient_Outdoors_Athena}}
 {id=SoundClass_4, children [Enveloper_5], properties {soundClassOverride=&Fort_Gadgets}}
 {id=SoundClass_5, children [Enveloper_4], properties {soundClassOverride=&Fort_Ambient_Outdoors_Athena}}
 {
  id=WavePlayer_0
  properties {bLooping=true, soundWaveAssetPtr=/Game/Sounds/Asteria/Environment_Actors/Fortnitanium/Spot_Fornitanium_Glowing_Rock_Loop}
 }
 {id=WavePlayer_1, properties {soundWaveAssetPtr=&/Spot_SkyIsland_Rumble_Oneshot_01}}
 {id=WavePlayer_10, properties {bLooping=true, soundWaveAssetPtr=&/Spot_SkyIsland_Metal_Rattles_Loop}}
 {id=WavePlayer_11, properties {soundWaveAssetPtr=/Game/Sounds/Artemis/Biomes/Cities/City_Spots/Spot_Cities_Metal_Rattles_01}}
 {id=WavePlayer_12, properties {soundWaveAssetPtr=/Game/Sounds/Artemis/Biomes/Cities/City_Spots/Spot_Cities_Metal_Rattles_02}}
 {id=WavePlayer_13, properties {soundWaveAssetPtr=/Game/Sounds/Artemis/Biomes/Cities/City_Spots/Spot_Cities_Metal_Rattles_03}}
 {id=WavePlayer_14, properties {soundWaveAssetPtr=/Game/Sounds/Artemis/Biomes/Cities/City_Spots/Spot_Cities_Metal_Rattles_04}}
 {id=WavePlayer_15, properties {soundWaveAssetPtr=/Game/Sounds/Artemis/Biomes/Cities/City_Spots/Spot_Cities_Metal_Rattles_05}}
 {id=WavePlayer_16, properties {soundWaveAssetPtr=&/Spot_WireStress_05}}
 {id=WavePlayer_17, properties {soundWaveAssetPtr=&/Spot_WireStress_06}}
 {id=WavePlayer_18, properties {soundWaveAssetPtr=&/Spot_WireStress_07}}
 {id=WavePlayer_19, properties {soundWaveAssetPtr=&/Spot_WireStress_08}}
 {id=WavePlayer_2, properties {soundWaveAssetPtr=&/Spot_SkyIsland_Rumble_Oneshot_02}}
 {
  id=WavePlayer_20
  properties {bLooping=true, soundWaveAssetPtr=&Spot_Fornitanium_Glowing_Rock_Loop}
 }
 {
  id=WavePlayer_21
  properties {bLooping=true, soundWaveAssetPtr=&/Spot_SkyIsland_Rumble_Sub_Presence_Loop}
 }
 {id=WavePlayer_22, properties {soundWaveAssetPtr=/Game/Sounds/Artemis/Biomes/Mountains/Mountain_Spots/Spot_Mountain_Rock_Slide_01}}
 {id=WavePlayer_23, properties {soundWaveAssetPtr=/Game/Sounds/Artemis/Biomes/Mountains/Mountain_Spots/Spot_Mountain_Rock_Slide_02}}
 {id=WavePlayer_24, properties {soundWaveAssetPtr=/Game/Sounds/Artemis/Biomes/Mountains/Mountain_Spots/Spot_Mountain_Rock_Slide_03}}
 {id=WavePlayer_25, properties {soundWaveAssetPtr=/Game/Sounds/Artemis/Biomes/Mountains/Mountain_Spots/Spot_Mountain_Rock_Slide_04}}
 {id=WavePlayer_26, properties {soundWaveAssetPtr=&/Spot_WireStress_09}}
 {id=WavePlayer_29, properties {soundWaveAssetPtr=/Game/Sounds/Fort_Ambient_NonLoop/Desert/Cliff_Dirt_01}}
 {id=WavePlayer_3, properties {soundWaveAssetPtr=&/Spot_WireStress_01}}
 {id=WavePlayer_30, properties {soundWaveAssetPtr=/Game/Sounds/Fort_Ambient_NonLoop/Desert/Cliff_Dirt_02}}
 {id=WavePlayer_31, properties {soundWaveAssetPtr=/Game/Sounds/Fort_Ambient_NonLoop/Desert/Cliff_Dirt_03}}
 {id=WavePlayer_32, properties {soundWaveAssetPtr=/Game/Sounds/Fort_Ambient_NonLoop/Desert/Cliff_Dirt_04}}
 {id=WavePlayer_33, properties {soundWaveAssetPtr=/Game/Sounds/Fort_Ambient_NonLoop/Desert/Cliff_Dirt_05}}
 {id=WavePlayer_34, properties {soundWaveAssetPtr=/Game/Sounds/Fort_Ambient_NonLoop/Desert/Cliff_Sand_01}}
 {id=WavePlayer_35, properties {soundWaveAssetPtr=/Game/Sounds/Fort_Ambient_NonLoop/Desert/Cliff_Sand_02}}
 {id=WavePlayer_36, properties {soundWaveAssetPtr=/Game/Sounds/Fort_Ambient_NonLoop/Desert/Cliff_Sand_03}}
 {id=WavePlayer_37, properties {soundWaveAssetPtr=/Game/Sounds/Fort_Ambient_NonLoop/Desert/Cliff_Sand_04}}
 {id=WavePlayer_38, properties {soundWaveAssetPtr=/Game/Sounds/Fort_Ambient_NonLoop/Desert/Cliff_Stones_01}}
 {id=WavePlayer_39, properties {soundWaveAssetPtr=/Game/Sounds/Fort_Ambient_NonLoop/Desert/Cliff_Stones_02}}
 {id=WavePlayer_4, properties {soundWaveAssetPtr=&/Spot_WireStress_03}}
 {id=WavePlayer_40, properties {soundWaveAssetPtr=/Game/Sounds/Fort_Ambient_NonLoop/Desert/Cliff_Stones_03}}
 {id=WavePlayer_41, properties {soundWaveAssetPtr=/Game/Sounds/Fort_Ambient_NonLoop/Desert/Cliff_Stones_04}}
 {id=WavePlayer_5, properties {soundWaveAssetPtr=&/Spot_SkyIsland_Rumble_Oneshot_04}}
 {id=WavePlayer_6, properties {soundWaveAssetPtr=&/Spot_SkyIsland_Rumble_Oneshot_05}}
 {id=WavePlayer_7, properties {soundWaveAssetPtr=&/Spot_WireStress_04}}
 {
  id=WavePlayer_8
  properties {bLooping=true, soundWaveAssetPtr=&Spot_SkyIsland_Rumble_Sub_Presence_Loop}
 }
 {
  id=WavePlayer_9
  properties {bLooping=true, soundWaveAssetPtr=&/Spot_SkyIsland_Rock_Crumble_Mids_Loop}
 }
]
properties {
 attenuationSettings=&/Spot_Fortnitanium_SkyIsland_Underside_No_Occ_Cone_Att
 bOverrideConcurrency=true
 maxDistance=20000
 soundClassObject=/Game/Sounds/FortSoundClassesAndModes/Fort_Ambient_Global
 volumeMultiplier=0.251
 attenuationOverrides {
  bEnableFocusInterpolation=true
  bEnableListenerFocus=true
  falloffDistance=500
  focusAttackInterpSpeed=0.01
  focusReleaseInterpSpeed=0.01
  focusVolumeAttenuation=1.2
  nonFocusVolumeAttenuation=0.75
  attenuationShapeExtents=<250 500>
 }
 concurrencyOverrides {maxCount=2, volumeScaleMode=Distance}
 preEffectBusSends [{sendLevel=0}]
}
